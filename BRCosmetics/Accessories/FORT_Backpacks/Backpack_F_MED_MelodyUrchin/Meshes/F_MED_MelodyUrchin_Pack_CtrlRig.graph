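# Control Rig for the MelodyUrchin backpack that procedurally animates its tongue and teeth bones.
# From Begin Execution, a per-item chain harmonics node drives the tongue from tongue_00 with wave and pendulum motion.
# Seven item harmonics nodes then follow in sequence, giving each tooth bone its own faster oscillation.
# Other tongue harmonics, random-lerp and bone-offset nodes sit unconnected to that execution path.
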ControlRigBlueprint /BRCosmetics/Accessories/FORT_Backpacks/Backpack_F_MED_MelodyUrchin/Meshes/F_MED_MelodyUrchin_Pack_CtrlRig
graphs [
 RigVMFunctionLibrary {links [], nodes []}
 RigVMModel {nodes [
   /Script/ControlRig.RigUnit_ItemHarmonics(
    ExecutePin=()
    RotationOrder=YZX
    Targets=((Item=(Type=Bone,Name="tongue_03"),Ratio=0))
    WaveAmplitude=(X=-5,Y=-45,Z=-5)
    WaveEase=SineEaseIn
    WaveFrequency=(X=1,Y=0.6,Z=0.8)
    WaveMaximum=1
    WaveMinimum=0.5
    WaveNoise=(X=0,Y=0,Z=0)
    WaveOffset=(X=2,Y=1,Z=2)
    WaveSpeed=(X=0.6,Y=0.6,Z=0.6)
   ) {is=struct RigVMUnitNode}
   > &RigUnit_ItemHarmonics(
    RotationOrder=YZX
    Targets=((Item=(Type=Bone,Name="tongue_02"),Ratio=0))
    WaveAmplitude=(X=20,Y=-40,Z=20)
    WaveEase=Linear
    WaveFrequency=(X=1,Y=0.6,Z=0.8)
    WaveMaximum=1
    WaveMinimum=0.5
    WaveNoise=(X=0,Y=0,Z=0)
    WaveOffset=(X=3,Y=0,Z=1)
    WaveSpeed=(X=0.8,Y=0.8,Z=0.8)
   )
   > &RigUnit_ItemHarmonics(
    RotationOrder=YZX
    Targets=((Item=(Type=Bone,Name="tongue_01"),Ratio=0))
    WaveAmplitude=(X=3,Y=10,Z=3)
    WaveEase=Linear
    WaveFrequency=(X=1,Y=0.6,Z=0.8)
    WaveMaximum=1
    WaveMinimum=0.5
    WaveNoise=(X=0,Y=0,Z=0)
    WaveOffset=(X=0,Y=1,Z=2)
    WaveSpeed=(X=0.85,Y=0.85,Z=0.85)
   )
   RigUnit_BeginExecution = &RigUnit_BeginExecution()
   /Script/ControlRig.RigUnit_ChainHarmonicsPerItem(
    ChainRoot=(Type=Bone,Name="tongue_00")
    DrawWorldOffset=(Rotation=(X=0,Y=0,Z=0,W=1),Translation=(X=0,Y=0,Z=0),Scale3D=(X=1,Y=1,Z=1))
    Pendulum=(bEnabled=True,PendulumStiffness=2,PendulumGravity=(X=0,Y=0,Z=0),PendulumBlend=0.75,PendulumDrag=0.98,PendulumMaximum=0.1,UnwindAxis=(X=0,Y=1,Z=0),UnwindMinimum=0.2,UnwindMaximum=0.05)
    Reach=(bEnabled=True,ReachTarget=(X=0,Y=0,Z=0),ReachAxis=(X=1,Y=0,Z=0),ReachMinimum=0,ReachMaximum=0)
    Speed=(X=0.8,Y=1.5,Z=1)
    Wave=(bEnabled=True,WaveFrequency=(X=-1,Y=-1,Z=-4.5),WaveAmplitude=(X=1,Y=2,Z=1),WaveOffset=(X=0,Y=0,Z=-3),WaveNoise=(X=0,Y=0,Z=0),WaveMinimum=-0.5,WaveMaximum=-1.25)
    WaveCurve="(EditorCurveData=(Keys=((Time=0.1619,Value=0.3394),(InterpMode=RCIM_Cubic,Time=0.3504,Value=0.3613,ArriveTangent=0.2483,LeaveTangent=0.2483),(InterpMode=RCIM_Cubic,Time=0.4991,Value=0.4231,ArriveTangent=0.5281,LeaveTangent=0.5281),(InterpMode=RCIM_Cubic,Ti ...(truncated)"
    bDrawDebug=True
    ExecutePin=@RigUnit_BeginExecution.ExecutePin
   ) {is=struct RigVMUnitNode}
   > &RigUnit_ItemHarmonics(
    RotationOrder=YZX
    Targets=((Item=(Type=Bone,Name="teeth_01"),Ratio=0))
    WaveAmplitude=(X=0,Y=50,Z=0)
    WaveEase=SineEaseIn
    WaveFrequency=(X=3,Y=3.6,Z=3.8)
    WaveMaximum=1
    WaveMinimum=0.5
    WaveNoise=(X=0,Y=0,Z=0)
    WaveOffset=(X=0,Y=3,Z=5)
    WaveSpeed=(X=3,Y=3,Z=3)
   )
   > &RigUnit_ItemHarmonics(
    RotationOrder=YZX
    Targets=((Item=(Type=Bone,Name="teeth_02"),Ratio=0))
    WaveAmplitude=(X=20,Y=0,Z=40)
    WaveEase=Linear
    WaveFrequency=(X=3,Y=3.6,Z=3.8)
    WaveMaximum=1
    WaveMinimum=0.5
    WaveNoise=(X=0,Y=0,Z=0)
    WaveOffset=(X=3,Y=0,Z=1)
    WaveSpeed=(X=3,Y=3,Z=3)
   )
   > &RigUnit_ItemHarmonics(
    RotationOrder=YZX
    Targets=((Item=(Type=Bone,Name="teeth_03"),Ratio=0))
    WaveAmplitude=(X=12,Y=20,Z=80)
    WaveEase=Linear
    WaveFrequency=(X=3,Y=3.6,Z=3.8)
    WaveMaximum=1
    WaveMinimum=0.5
    WaveNoise=(X=0,Y=0,Z=0)
    WaveOffset=(X=0,Y=1,Z=2)
    WaveSpeed=(X=3,Y=3,Z=3)
   )
   > &RigUnit_ItemHarmonics(
    RotationOrder=YZX
    Targets=((Item=(Type=Bone,Name="teeth_06"),Ratio=0))
    WaveAmplitude=(X=12,Y=40,Z=10)
    WaveEase=Linear
    WaveFrequency=(X=3,Y=3.6,Z=3.8)
    WaveMaximum=1
    WaveMinimum=0.5
    WaveNoise=(X=0,Y=0,Z=0)
    WaveOffset=(X=0,Y=1,Z=2)
    WaveSpeed=(X=3,Y=3,Z=3)
   )
   > &RigUnit_ItemHarmonics(
    RotationOrder=YZX
    Targets=((Item=(Type=Bone,Name="teeth_07"),Ratio=0))
    WaveAmplitude=(X=12,Y=15,Z=60)
    WaveEase=Linear
    WaveFrequency=(X=3,Y=3.6,Z=3.8)
    WaveMaximum=1
    WaveMinimum=0.5
    WaveNoise=(X=0,Y=0,Z=0)
    WaveOffset=(X=0,Y=-10,Z=4)
    WaveSpeed=(X=3,Y=3,Z=3)
   )
   > &RigUnit_ItemHarmonics(
    RotationOrder=YZX
    Targets=((Item=(Type=Bone,Name="teeth_04"),Ratio=0))
    WaveAmplitude=(X=15,Y=3,Z=15)
    WaveEase=Linear
    WaveFrequency=(X=3,Y=3.6,Z=3.8)
    WaveMaximum=1
    WaveMinimum=0.5
    WaveNoise=(X=0,Y=0,Z=0)
    WaveOffset=(X=0,Y=2,Z=4)
    WaveSpeed=(X=3,Y=3,Z=3)
   )
   > &RigUnit_ItemHarmonics(
    RotationOrder=YZX
    Targets=((Item=(Type=Bone,Name="teeth_05"),Ratio=0))
    WaveAmplitude=(X=30,Y=3,Z=9)
    WaveEase=Linear
    WaveFrequency=(X=3,Y=3.6,Z=3.8)
    WaveMaximum=1
    WaveMinimum=0.5
    WaveNoise=(X=0,Y=0,Z=0)
    WaveOffset=(X=4,Y=15,Z=1)
    WaveSpeed=(X=3,Y=3,Z=3)
   )
   n1 = /Script/RigVM.RigVMFunction_RandomFloat(Duration=3, Maximum=1, Minimum=0, Seed=217) {is=struct RigVMUnitNode, template="Random::Execute(in Seed,in Minimum,in Maximum,in Duration,out Result)"}
   /Script/RigVM.RigVMFunction_AccumulateFloatLerp(
    Blend=14
    InitialValue=1
    bIntegrateDeltaTime=true
    TargetValue=&RigVMFunction_MathFloatRemap(SourceMaximum=1, SourceMinimum=0, TargetMaximum=0.3, TargetMinimum=0, bClamp=False, Value=@n1.Result).Result
   ) {is=struct RigVMUnitNode, template="AccumulateLerp::Execute(in TargetValue,in InitialValue,in Blend,in bIntegrateDeltaTime,out Result)"}
   &RigVMFunction_AccumulateFloatLerp(
    Blend=14
    InitialValue=1
    bIntegrateDeltaTime=true
    TargetValue=&RigVMFunction_MathFloatRemap(SourceMaximum=1, SourceMinimum=0, TargetMaximum=0.3, TargetMinimum=0, bClamp=False, Value=@n1.Result).Result
   ) {template="AccumulateLerp::Execute(in TargetValue,in InitialValue,in Blend,in bIntegrateDeltaTime,out Result)"}
   /Script/ControlRig.RigUnit_ChainHarmonics(
    ChainRoot=tongue_01
    DrawWorldOffset=(Rotation=(X=0,Y=0,Z=0,W=1),Translation=(X=0,Y=0,Z=0),Scale3D=(X=1,Y=1,Z=1))
    ExecutePin=()
    Pendulum=(bEnabled=False,PendulumStiffness=0.1,PendulumGravity=(X=0,Y=0,Z=0),PendulumBlend=1,PendulumDrag=1,PendulumMinimum=0,PendulumMaximum=0.1,PendulumEase=Linear,UnwindAxis=(X=0,Y=1,Z=0),UnwindMinimum=0.2,UnwindMaximum=0.05)
    Reach=(bEnabled=False,ReachTarget=(X=0,Y=0,Z=0),ReachAxis=(X=1,Y=0,Z=0),ReachMinimum=0,ReachMaximum=0,ReachEase=SineEaseInOut)
    Speed=(X=0.25,Y=0.25,Z=0.1)
    Wave=(bEnabled=True,WaveFrequency=(X=0.2,Y=0.2,Z=0.2),WaveAmplitude=(X=0.25,Y=2.5,Z=1.5),WaveOffset=(X=1,Y=0,Z=2),WaveNoise=(X=8,Y=8,Z=30),WaveMinimum=0,WaveMaximum=0.75,WaveEase=SineEaseIn)
    WaveCurve=(EditorCurveData=(Keys=((Time=-0.0198,Value=0.4929),(Time=0.9855,Value=0.5049)),PreInfinityExtrap=RCCE_Constant,PostInfinityExtrap=RCCE_Constant,DefaultValue=340282346638528859811704183484516925440))
    bDrawDebug=True
   ) {is=struct RigVMUnitNode}
   &RigUnit_ChainHarmonicsPerItem(
    ChainRoot=(Type=Bone,Name="tongue_01")
    DrawWorldOffset=(Rotation=(X=0,Y=0,Z=0,W=1),Translation=(X=0,Y=0,Z=0),Scale3D=(X=1,Y=1,Z=1))
    ExecutePin=()
    Pendulum=(bEnabled=true,PendulumStiffness=1.5,PendulumGravity=(X=0,Y=0,Z=0),PendulumBlend=0.95,PendulumDrag=0.98,PendulumMinimum=0,PendulumMaximum=0.7,PendulumEase=Linear,UnwindAxis=(X=0,Y=1,Z=0),UnwindMinimum=0.2,UnwindMaximum=0.05)
    Reach=(bEnabled=true,ReachTarget=(X=1.5,Y=1.5,Z=1.5),ReachAxis=(X=1,Y=0,Z=0),ReachMinimum=0,ReachMaximum=0,ReachEase=Linear)
    Speed=(X=3,Y=3,Z=3)
    Wave=(bEnabled=True,WaveFrequency=(X=0,Y=0,Z=0),WaveAmplitude=(X=0,Y=8,Z=0),WaveOffset=(X=0,Y=0,Z=0),WaveNoise=(X=0,Y=0,Z=0),WaveMinimum=5,WaveMaximum=5,WaveEase=SineEaseIn)
    WaveCurve="(EditorCurveData=(Keys=((InterpMode=RCIM_Cubic,Time=0.0121,Value=0.0862),(InterpMode=RCIM_Cubic,Time=0.3381,Value=0.2251,ArriveTangent=0.1065,LeaveTangent=0.1065),(InterpMode=RCIM_Cubic,Time=0.6082,Value=0.1497,ArriveTangent=-0.1559,LeaveTangent=-0.1559),( ...(truncated)"
    bDrawDebug=True
   )
   &RigUnit_ChainHarmonics(
    ChainRoot=tongue_01
    DrawWorldOffset=(Rotation=(X=0,Y=0,Z=0,W=1),Translation=(X=0,Y=0,Z=0),Scale3D=(X=1,Y=1,Z=1))
    ExecutePin=()
    Pendulum=(bEnabled=True,PendulumStiffness=2,PendulumGravity=(X=0,Y=0,Z=0),PendulumBlend=0.75,PendulumDrag=0.98,PendulumMaximum=0.1,UnwindAxis=(X=0,Y=1,Z=0),UnwindMinimum=0.2,UnwindMaximum=0.05)
    Reach=(bEnabled=false,ReachTarget=(X=0,Y=0,Z=0),ReachAxis=(X=1,Y=0,Z=0),ReachMinimum=0)
    Speed=(X=1,Y=1,Z=1)
    Wave=(bEnabled=True,WaveFrequency=(X=1,Y=0.6,Z=0.8),WaveAmplitude=(X=1,Y=0.6,Z=0.8),WaveOffset=(X=0,Y=1,Z=0),WaveNoise=(X=0,Y=0,Z=0),WaveMaximum=1)
    WaveCurve=(EditorCurveData=(Keys=((),(Time=1,Value=1)),DefaultValue=340282346638528859811704183484516925440,PreInfinityExtrap=RCCE_Constant,PostInfinityExtrap=RCCE_Constant))
    bDrawDebug=True
   )
   &RigUnit_ChainHarmonics(
    ChainRoot=tongue_01
    DrawWorldOffset=(Rotation=(X=0,Y=0,Z=0,W=1),Translation=(X=0,Y=0,Z=0),Scale3D=(X=1,Y=1,Z=1))
    ExecutePin=()
    Pendulum=(bEnabled=False,PendulumStiffness=2,PendulumGravity=(X=0,Y=0,Z=0),PendulumBlend=0.75,PendulumDrag=0.98,PendulumMinimum=0,PendulumMaximum=0.1,PendulumEase=Linear,UnwindAxis=(X=0,Y=1,Z=0),UnwindMinimum=0.2,UnwindMaximum=0.05)
    Reach=(bEnabled=False,ReachTarget=(X=0,Y=0,Z=0),ReachAxis=(X=1,Y=0,Z=0),ReachMinimum=0,ReachMaximum=0,ReachEase=SineEaseInOut)
    Speed=(X=0.25,Y=0.25,Z=0.1)
    Wave=(bEnabled=True,WaveFrequency=(X=0.2,Y=0.2,Z=0.2),WaveAmplitude=(X=0.25,Y=2.5,Z=1.5),WaveOffset=(X=1,Y=0,Z=2),WaveNoise=(X=8,Y=8,Z=30),WaveMinimum=0.1,WaveMaximum=0.5,WaveEase=SineEaseIn)
    WaveCurve=(EditorCurveData=(Keys=((Time=-0.0198,Value=0.4929),(Time=0.9966,Value=0.5016))))
    bDrawDebug=True
   )
   /Script/ControlRig.RigUnit_ModifyBoneTransforms(
    BoneToModify=((Bone="tongue_01",Transform=(Rotation=(X=0,Y=0,Z=0,W=1),Translation=(X=0,Y=0,Z=0),Scale3D=(X=1,Y=1,Z=1))),(Bone="tongue_02",Transform=(Rotation=(X=0,Y=0,Z=0,W=1),Translation=(X=2,Y=0,Z=0),Scale3D=(X=1,Y=1,Z=1))))
    ExecutePin=()
    Mode=AdditiveLocal
    Weight=1
    WeightMaximum=1
    WeightMinimum=0
    BoneToModify.0.Transform.Rotation=&RigVMFunction_MathQuaternionFromEuler(Euler=(X=0,Y=0,Z=-10)).Result
   ) {is=struct RigVMUnitNode}
  ]}
]
hierarchy [
 root {loc=<0 0 0>, rot=<0 0 0 1>}
 pack {parent=root, loc=<0 -0.7316 0>, rot=<0 0 0 1>}
 teeth_01 {parent=pack, loc=<-0.777 -13.07 0.5015>, rot=<0 0 0.7071 -0.7071>}
 teeth_02 {parent=pack, loc=<6.7714 -12.1566 -1.9243>, rot=<0 0 0.7071 -0.7071>}
 teeth_03 {parent=pack, loc=<-5.9461 -13.2688 -5.0057>, rot=<0 0 0.7071 -0.7071>}
 teeth_04 {parent=pack, loc=<6.6797 -12.3671 -13.073>, rot=<0 0 0.7071 -0.7071>}
 teeth_05 {parent=pack, loc=<-4.975 -14.4498 -13.3784>, rot=<0 0 0.7071 -0.7071>}
 handel {parent=pack, loc=<0 -2.4938 12.7777>, rot=<0.5 0.5 0.5 -0.5>}
 zipper {parent=pack, loc=<-11.4604 -5.4845 -0.3071>, rot=<0 -0.7817 0 -0.6236>}
 teeth_06 {parent=pack, loc=<3.2006 -12.1566 -0.1218>, rot=<0 0 0.7071 -0.7071>}
 teeth_07 {parent=pack, loc=<-5.3704 -13.2533 -3.3236>, rot=<0 0 0.7071 -0.7071>}
 tongue_00 {parent=root, loc=<1.2232 -14.9603 -9.9986>, rot=<-0.206 -0.2442 0.6392 -0.6996>}
 tongue_01 {parent=tongue_00, loc=<5.2139 -0.0375 0.0882>, rot=<-0.0173 -0.4635 0.033 -0.8853>}
 tongue_02 {parent=tongue_01, loc=<7.7224 0 0>, rot=<0 0 0 1>}
 tongue_03 {parent=tongue_02, loc=<6.4506 0 0>, rot=<0 0 0 1>}
 tongue_04 {parent=tongue_03, loc=<3.3326 0.3208 0.5586>, rot=<0.0093 0.1533 -0.06 -0.9863>}
]
