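# Physics asset for the MelodyUrchin backpack cosmetic, defining collision bodies and joint limits for the pack, its handle and a dangling tongue.
# The pack combines a capsule and three boxes, hinged to the root, with the handle almost rigidly constrained to it.
# Five tongue capsules chain from the root rather than the pack, with almost no twist and swing limits that widen along the chain before tightening at the last joint.
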PhysicsAsset /BRCosmetics/Accessories/FORT_Backpacks/Backpack_F_MED_MelodyUrchin/Meshes/F_MED_MelodyUrchin_Pack_Physics
bodies {
 handel [{shapeType=Capsule, length=1.1905, radius=0.9538, center=<0.9522 0 0>}]
 pack [
  {shapeType=Capsule, length=5.9873, radius=11.4835, center=<1.0628 -4.6575 -6.2984>}
  {shapeType=Box, extentX=10.5, extentY=10.5, extentZ=13.5, shapeName=pack_box, center=<-6.9114 -8.7866 -17.6044>}
  {shapeType=Box, extentX=10.5, extentY=11.5, extentZ=13.5, shapeName=pack_box, center=<8.7187 -6.3624 -17.556>}
  {shapeType=Box, extentX=19.5, extentY=7.5, extentZ=21.5, shapeName=pack_box, center=<0 -7.9301 -18.8495>}
 ]
 root [{shapeType=Capsule, length=1, radius=1, rotation=<rot 90 0 0>}]
 tongue_00 [{shapeType=Capsule, length=3.0501, radius=2.2083, shapeName=tongue_01_capsule, center=<4.0159 0 0>, rotation=<rot 90 0 0>}]
 tongue_01 [{shapeType=Capsule, length=3.0501, radius=2.2083, shapeName=tongue_01_capsule, center=<4.0159 0 0>, rotation=<rot 90 0 0>}]
 tongue_02 [{shapeType=Capsule, length=3.4275, radius=2.2083, shapeName=tongue_02_capsule, center=<3.2774 0 0>, rotation=<rot 90 0 0>}]
 tongue_03 [{shapeType=Capsule, length=3.4275, radius=2.2083, shapeName=tongue_02_capsule, center=<3.9633 0 0>, rotation=<rot 90 0 0>}]
 tongue_04 [{shapeType=Capsule, length=3.4275, radius=2.2083, shapeName=tongue_02_capsule, center=<4.0869 0 0>, rotation=<rot 90 0 0>}]
}
constraints [
 {bone1Name=pack, bone2Name=root, swing1LimitDegrees=45, swing2LimitDegrees=45, twistLimitDegrees=45}
 {bone1Name=tongue_01, bone2Name=tongue_00, swing1LimitDegrees=30, swing2LimitDegrees=50, twistLimitDegrees=1}
 {bone1Name=tongue_02, bone2Name=tongue_01, swing1LimitDegrees=45, swing2LimitDegrees=50, twistLimitDegrees=1}
 {bone1Name=tongue_03, bone2Name=tongue_02, swing1LimitDegrees=60, swing2LimitDegrees=70, twistLimitDegrees=1}
 {bone1Name=handel, bone2Name=pack, swing1LimitDegrees=1, swing2LimitDegrees=1, twistLimitDegrees=0}
 {bone1Name=tongue_00, bone2Name=root, swing1LimitDegrees=3, swing2LimitDegrees=15, twistLimitDegrees=1}
 {bone1Name=tongue_04, bone2Name=tongue_03, swing1LimitDegrees=6, swing2LimitDegrees=12, twistLimitDegrees=1}
]
