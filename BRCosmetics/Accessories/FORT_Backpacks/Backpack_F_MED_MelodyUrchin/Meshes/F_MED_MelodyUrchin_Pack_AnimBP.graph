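# Animation Blueprint for the MelodyUrchin backpack cosmetic, adding secondary motion to its tongue, zipper and other pack bones.
# bRigidBodyEnabled selects RigidBody physics over a simple tongue_04 trail.
# bIsSkyDiving picks between two RigidBody branches with Control Rig: one rotates tongue bones from spine-derived Backpack Rotate, the other adds zipper dynamics and fixed tongue offsets.
# Several teeth, tongue and handle dynamics nodes remain disconnected from the output.
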AnimBlueprint /BRCosmetics/Accessories/FORT_Backpacks/Backpack_F_MED_MelodyUrchin/Meshes/F_MED_MelodyUrchin_Pack_AnimBP
target_skeleton=&/F_MED_MelodyUrchin_Pack_Skeleton
graphs [
 AnimGraph {nodes [
   Anim Dynamics - Bone: handel(Alpha=1, ExternalForce="0,0,0") {
    angularDampingOverride=0.8
    angularSpringConstant=200
    bAngularSpring=true
    bGravityOverrideInSimSpace=true
    bLinearSpring=true
    bOverrideAngularDamping=true
    bOverrideLinearDamping=true
    gravityScale=0.35
    linearDampingOverride=0.8
    linearSpringConstant=300
    alphaBoolBlend {blendOption=Linear}
    boundBone {boneName=handel}
    componentAppliedLinearAccClamp=<400 400 500>
    componentLinearAccScale=<0.1 0.1 0.1>
    componentLinearVelScale=<0.1 0.1 0.1>
    gravityOverride {z=-980}
    physicsBodyDefinitions [{
      sphereCollisionRadius=0
      boundBone {boneName=handel}
      boxExtents=<20 20 5>
      constraintSetup {angularLimitsMax=<15 5 1>, angularLimitsMin=<-15 -5 -1>, angularTarget=<0 0 0>, linearAxesMax=<0 0 0.5>, linearAxesMin=<0 0 -0.5>}
      localJointOffset=<-0.1 -0.1 2>
     }]
   }
   n1 = Anim Dynamics - Bone: tongue_01(Alpha=0.5, ExternalForce="0,0,0") {
    angularDampingOverride=0.8
    angularSpringConstant=3000
    bAngularSpring=true
    bGravityOverrideInSimSpace=true
    bOverrideAngularDamping=true
    gravityScale=0.35
    simulationSpace=RootRelative
    alphaBoolBlend {blendOption=Linear}
    boundBone {boneName=tongue_01}
    componentAppliedLinearAccClamp=<4000 4000 4000>
    componentLinearAccScale=<0.8 0.8 0.8>
    componentLinearVelScale=<0.4 0.4 0.4>
    physicsBodyDefinitions [{
      sphereCollisionRadius=0
      boundBone {boneName=tongue_01}
      boxExtents=<10 10 10>
      constraintSetup {angularLimitsMax=<0 1 100>, angularLimitsMin=<0 -1 -40>, angularTarget=<1 0 0>}
      localJointOffset=<1 0 0>
     }]
   }
   n2 = Anim Dynamics - Bone: tongue_02(ComponentPose=@n1, Alpha=1, ExternalForce="0,0,0") {
    angularDampingOverride=0.9
    angularSpringConstant=2000
    bAngularSpring=true
    bGravityOverrideInSimSpace=true
    bOverrideAngularDamping=true
    gravityScale=0.35
    simulationSpace=RootRelative
    alphaBoolBlend {blendOption=Linear}
    alphaScaleBiasClamp {bInterpResult=true, interpSpeedDecreasing=5, interpSpeedIncreasing=7}
    boundBone {boneName=tongue_02}
    componentAppliedLinearAccClamp=<5000 5000 5000>
    componentLinearAccScale=<0.8 0.8 0.8>
    componentLinearVelScale=<0.3 0.3 0.3>
    physicsBodyDefinitions [{
      sphereCollisionRadius=0
      boundBone {boneName=tongue_02}
      boxExtents=<10 10 10>
      constraintSetup {angularLimitsMax=<1 1 100>, angularLimitsMin=<-1 -2 -50>, angularTarget=<1 0 0>}
      localJointOffset=<1 0 0>
     }]
   }
   Anim Dynamics - Bone: tongue_03(ComponentPose=@n2, Alpha=1, ExternalForce="0,0,0") {
    angularDampingOverride=0.999
    angularSpringConstant=3000
    bAngularSpring=true
    bGravityOverrideInSimSpace=true
    bOverrideAngularDamping=true
    gravityScale=0.35
    simulationSpace=RootRelative
    alphaBoolBlend {blendOption=Linear}
    alphaScaleBiasClamp {bInterpResult=true, interpSpeedDecreasing=5, interpSpeedIncreasing=7}
    boundBone {boneName=tongue_03}
    componentLinearAccScale=<0.8 0.8 0.8>
    componentLinearVelScale=<0.4 0.4 0.4>
    physicsBodyDefinitions [{
      sphereCollisionRadius=0
      boundBone {boneName=tongue_03}
      boxExtents=<10 10 10>
      constraintSetup {angularLimitsMax=<1 1 5>, angularLimitsMin=<-1 -1 -20>, angularTarget=<1 0 0>}
      localJointOffset=<1 0 0>
     }]
   }
   n3 = Anim Dynamics - Bone: teeth_01(Alpha=0.7, ExternalForce="0,0,0") {
    angularDampingOverride=0.8
    angularSpringConstant=9000
    bGravityOverrideInSimSpace=true
    bOverrideAngularDamping=true
    gravityScale=0.35
    simulationSpace=RootRelative
    alphaBoolBlend {blendOption=Linear}
    alphaScaleBiasClamp {interpSpeedDecreasing=5, interpSpeedIncreasing=7}
    boundBone {boneName=teeth_01}
    componentAppliedLinearAccClamp=<5000 5000 5000>
    componentLinearAccScale=<0.4 0.4 0.4>
    componentLinearVelScale=<0.2 0.2 0.2>
    physicsBodyDefinitions [{
      sphereCollisionRadius=0
      boundBone {boneName=teeth_01}
      boxExtents=<10 10 10>
      constraintSetup {angularLimitsMax=<0 15 2>, angularLimitsMin=<0 -30 -2>, angularTarget=<1 0 0>}
      localJointOffset=<1 0 0>
     }]
   }
   n4 = Anim Dynamics - Bone: teeth_02(ComponentPose=@n3, Alpha=0.5, ExternalForce="0,0,0") {
    angularDampingOverride=0.8
    angularSpringConstant=9000
    bGravityOverrideInSimSpace=true
    bOverrideAngularDamping=true
    gravityScale=0.35
    simulationSpace=RootRelative
    alphaBoolBlend {blendOption=Linear}
    alphaScaleBiasClamp {interpSpeedDecreasing=5, interpSpeedIncreasing=7}
    boundBone {boneName=teeth_02}
    componentAppliedLinearAccClamp=<5000 5000 5000>
    componentLinearAccScale=<0.5 0.5 0.5>
    componentLinearVelScale=<0.2 0.2 0.2>
    physicsBodyDefinitions [{
      sphereCollisionRadius=0
      boundBone {boneName=teeth_02}
      boxExtents=<10 10 10>
      constraintSetup {angularLimitsMax=<0 15 15>, angularLimitsMin=<0 -25 -2>, angularTarget=<1 0 0>}
      localJointOffset=<1 0 0>
     }]
   }
   Anim Dynamics - Bone: teeth_03(ComponentPose=@n4, Alpha=0.4, ExternalForce="0,0,0") {
    angularDampingOverride=0.8
    angularSpringConstant=9000
    bGravityOverrideInSimSpace=true
    bOverrideAngularDamping=true
    gravityScale=0.35
    simulationSpace=RootRelative
    alphaBoolBlend {blendOption=Linear}
    alphaScaleBiasClamp {interpSpeedDecreasing=5, interpSpeedIncreasing=7}
    boundBone {boneName=teeth_03}
    componentAppliedLinearAccClamp=<5000 5000 5000>
    componentLinearAccScale=<0.6 0.6 0.6>
    componentLinearVelScale=<0.2 0.2 0.2>
    physicsBodyDefinitions [{
      sphereCollisionRadius=0
      boundBone {boneName=teeth_03}
      boxExtents=<10 10 10>
      constraintSetup {angularLimitsMax=<0 15 2>, angularLimitsMin=<0 -20 -15>, angularTarget=<1 0 0>}
      localJointOffset=<1 0 0>
     }]
   }
   Component To Local()
   Copy Pose From Mesh()
   Local To Component()
   Local To Component()
   "Transform (Modify) Bone - Bone: tongue_03"(Alpha=0.3, Translation="0,0,0", Rotation="0,0,0", Scale="1,1,1") {rotationMode=BMM_Additive, rotationSpace=BCS_BoneSpace, alphaBoolBlend {blendOption=Linear}, boneToModify {boneName=tongue_03}}
   n5 = RigidBody(ComponentPose=Local To Component(LocalPose=Use cached pose 'prerigid'())) {
    bClampLinearTranslationLimitToRefPose=true
    bOverrideWorldGravity=true
    overridePhysicsAsset=&/F_MED_MelodyUrchin_Pack_Physics
    simulationSpace=BaseBoneSpace
    alphaBoolBlend {blendOption=Linear}
    baseBoneRef {boneName=pack}
    componentAppliedLinearAccClamp=<5000 5000 5000>
    componentLinearAccScale=<0.4 0.4 0.5>
    componentLinearVelScale=<0.25 0.25 0.25>
    overrideWorldGravity {z=-980}
   }
   n6 = Control Rig(Source=Component To Local(ComponentPose=@n5), Alpha=0.2) {
    alphaBoolBlend {blendOption=Linear}
    controlRigAssetReference {blueprintRigClass=&/F_MED_MelodyUrchin_Pack_CtrlRig.F_MED_MelodyUrchin_Pack_CtrlRig_C}
   }
   n7 = Get Backpack Rotate() -> (Backpack Rotate:Rotator)
   n8 = "Transform (Modify) Bone - Bone: tongue_01"(ComponentPose=Local To Component(LocalPose=@n6), Alpha=1, Translation="0,0,0", Rotation=@n7, Scale="1,1,1") {rotationMode=BMM_Additive, rotationSpace=BCS_BoneSpace, alphaBoolBlend {blendOption=Linear}, boneToModify {boneName=tongue_01}}
   n9 = "Transform (Modify) Bone - Bone: tongue_02"(ComponentPose=@n8, Alpha=0.5, Translation="0,0,0", Rotation=@n7, Scale="1,1,1") {
    rotationMode=BMM_Additive
    rotationSpace=BCS_BoneSpace
    alphaBoolBlend {blendOption=Linear}
    alphaScaleBiasClamp {bClampResult=true, clampMax=0.5}
    boneToModify {boneName=tongue_02}
   }
   n10 = Component To Local(ComponentPose=@n9)
   n11 = RigidBody(ComponentPose=Local To Component(LocalPose=Use cached pose 'prerigid'())) {
    bClampLinearTranslationLimitToRefPose=true
    bOverrideWorldGravity=true
    overridePhysicsAsset=&F_MED_MelodyUrchin_Pack_Physics
    simulationSpace=BaseBoneSpace
    alphaBoolBlend {blendOption=Linear}
    alphaScaleBiasClamp {interpSpeedDecreasing=1}
    baseBoneRef {boneName=pack}
    componentAppliedLinearAccClamp=<3000 3000 3000>
    componentLinearAccScale=<0.6 1 0.6>
    componentLinearVelScale=<0.25 0.5 0.25>
    overrideWorldGravity {z=-980}
   }
   n12 = Anim Dynamics - Bone: zipper(ComponentPose=@n11, Alpha=1, ExternalForce="0,0,0") {
    angularDampingOverride=1
    angularSpringConstant=1000
    bAngularSpring=true
    bOverrideAngularDamping=true
    bUseGravityOverride=true
    bUsePlanarLimit=false
    simulationSpace=RootRelative
    alphaBoolBlend {blendOption=Linear}
    boundBone {boneName=zipper}
    componentAppliedLinearAccClamp=<4000 4000 4000>
    componentLinearAccScale=<0.7 0.7 0.7>
    componentLinearVelScale=<0.7 0.7 0.7>
    gravityOverride {z=-980}
    physicsBodyDefinitions [{
      boundBone {boneName=zipper}
      boxExtents=<10 10 10>
      constraintSetup {angularLimitsMax=<15 90 30>, angularLimitsMin=<-15 -15 -30>, angularTarget=<0 0 0>}
      localJointOffset=<2 0.0517 -0.0157>
     }]
    relativeSpaceBone {boneName=pack}
   }
   n13 = Control Rig(Source=Component To Local(ComponentPose=@n12), Alpha=0.2) {
    alphaBoolBlend {blendOption=Linear}
    controlRigAssetReference {blueprintRigClass=&F_MED_MelodyUrchin_Pack_CtrlRig_C}
   }
   n14 = "Transform (Modify) Bone - Bone: tongue_01"!(
    ComponentPose=Local To Component(LocalPose=@n13)
    bAlphaBoolEnabled:bool=True
    Translation="0,0,0"
    Rotation="19.997,0.041,0.191"
    Scale="1,1,1"
   ) -> (Pose) {
    alphaInputType=Bool
    rotationMode=BMM_Additive
    rotationSpace=BCS_BoneSpace
    alphaBoolBlend {blendOption=Linear}
    boneToModify {boneName=tongue_01}
    rotation=<rot 19.9969 0.041 0.191>
   }
   n15 = "Transform (Modify) Bone - Bone: tongue_02"!(ComponentPose=@n14, bAlphaBoolEnabled:bool=True, Translation="0,0,0", Rotation="-50.001,-0.171,0.343", Scale="1,1,1") -> (Pose) {
    alphaInputType=Bool
    rotationMode=BMM_Additive
    rotationSpace=BCS_BoneSpace
    alphaBoolBlend {blendOption=Linear}
    boneToModify {boneName=tongue_02}
    rotation=<rot -50.001 -0.171 0.343>
   }
   n16 = Component To Local(ComponentPose=@n15)
   n17 = Get bIsSkyDiving()
   n18 = Trail controller - Bone: tongue_04(ComponentPose=Local To Component(LocalPose=Use cached pose 'prerigid'()), Alpha=0.8) {
    bLimitRotation=true
    chainLength=5
    relaxationSpeedScale=1.3
    planarLimits [{drivingBone {boneName=root}, planeTransform {location=<0 0 0>, rotation=<rot 0 0 0>, scale=<0 0 0>}}]
    rotationLimits [
     {limitMax=<90 10 10>, limitMin=<-90 -90 -90>}
     {limitMax=<90 10 90>, limitMin=<-90 -90 -90>}
     {limitMax=<90 15 90>, limitMin=<-90 -90 -90>}
     {limitMax=<90 15 90>, limitMin=<-90 -90 -90>}
     {limitMax=<180 180 180>}
    ]
    rotationOffsets [<0 0 0>, <0 0 0>, <0 0 0>, <0 0 0>, <0 0 0>]
    trailBone {boneName=tongue_04}
    trailRelaxationSpeed {editorCurveData {keys [<at 0 25 RCIM_Linear>,<at 1 5 RCIM_Linear>]}}
   }
   Output Pose(Result=Blend Poses by bool(
     BlendPose_0=Blend Poses by bool(BlendPose_0=@n10, BlendPose_1=@n16, bActiveValue=@n17) {blendType=Linear}
     BlendPose_1=Component To Local(ComponentPose=@n18)
     bActiveValue=Get bRigidBodyEnabled()
    ) {blendType=Linear}) {layerGroup=None}
   Save cached pose 'prerigid'(Pose=Input Pose()) {cacheName=prerigid}
   Save cached pose 'prerigid'() {cacheName=prerigid}
   Trail controller - Bone: tongue_00!(ComponentPose, bAlphaBoolEnabled:bool=True, RelaxationSpeedScale:float=1, FakeVelocity:Vector="1,1,1") -> (Pose) {
    alphaInputType=Bool
    bActorSpaceFakeVel=true
    bLimitRotation=true
    bShowTrailLocation=true
    bUsePlanarLimit=true
    chainLength=5
    debugLifeTime=0.1
    lODThreshold=1
    maxDeltaTime=0.0333
    stretchLimit=2
    alphaScaleBiasClamp {bInterpResult=true, interpSpeedDecreasing=25, interpSpeedIncreasing=25}
    baseJoint {boneName=pack}
    relaxationSpeedScaleInputProcessor {bias=0.3, scale=0.4}
    rotationLimits [
     {limitMax=<0 180 0>, limitMin=<0 -180 0>}
     {limitMax=<0 180 0>, limitMin=<0 -180 0>}
     {limitMax=<0 176.9904 0>, limitMin=<0 -180 0>}
     {limitMax=<0 180 0>, limitMin=<0 -180 0>}
     {limitMax=<180 180 180>}
    ]
    rotationOffsets [<0 0 0>, <0 0 0>, <0 0 0>, <0 0 0>, <0 0 0>]
    trailBone {boneName=tongue_00}
    trailRelaxationSpeed {editorCurveData {keys [<at 0 37.4961 RCIM_Linear>,<at 0.49 33.4961 RCIM_Linear>,<at 1 27.4961 RCIM_Linear>]}}
   }
   Trail controller - Bone: tongue_00(
    ComponentPose=Local To Component(LocalPose=Use cached pose 'prerigid'())
    bAlphaBoolEnabled=True
    RelaxationSpeedScale=1
    FakeVelocity="1,1,1"
   ) {
    alphaInputType=Bool
    bActorSpaceFakeVel=true
    bLimitRotation=true
    bShowTrailLocation=true
    bUsePlanarLimit=true
    chainLength=5
    debugLifeTime=0.1
    lODThreshold=1
    maxDeltaTime=0.0333
    stretchLimit=2
    baseJoint {boneName=root}
    rotationLimits [{limitMax=<180 180 180>}, {limitMax=<180 180 180>}, {limitMax=<180 180 180>}, {limitMax=<180 180 180>}, {limitMax=<180 180 180>}]
    rotationOffsets [<0 0 0>, <0 0 0>, <0 0 0>, <0 0 0>, <0 0 0>]
    trailBone {boneName=tongue_00}
    trailRelaxationSpeed {editorCurveData {keys [<at 0 10 RCIM_Linear>,<at 0.0642 9.5523 RCIM_Linear>,<at 0.5137 7.9619 RCIM_Linear>,<at 1.0205 5.083 RCIM_Linear>]}}
   }
   Trail controller - Bone: tongue_00(bAlphaBoolEnabled=True, RelaxationSpeedScale=1, FakeVelocity="1,1,1") {
    alphaInputType=Bool
    bActorSpaceFakeVel=true
    bLimitRotation=true
    bShowTrailLocation=true
    bUsePlanarLimit=true
    chainLength=4
    debugLifeTime=0.1
    lODThreshold=1
    maxDeltaTime=0.0333
    alphaScaleBiasClamp {bInterpResult=true, interpSpeedDecreasing=25, interpSpeedIncreasing=25}
    baseJoint {boneName=root}
    relaxationSpeedScaleInputProcessor {bias=0.3, scale=0.4}
    rotationLimits [
     {limitMax=<180 180 0>, limitMin=<0 -180 0>}
     {limitMax=<0 180 0>, limitMin=<0 -180 0>}
     {limitMax=<0 180 0>, limitMin=<0 -180 0>}
     {limitMax=<0 180 0>, limitMin=<0 -180 0>}
    ]
    rotationOffsets [<0 0 0>, <0 0 0>, <0 0 0>, <0 0 0>]
    trailBone {boneName=tongue_00}
    trailRelaxationSpeed {editorCurveData {keys [<at 0 37.1289 RCIM_Linear>,<at 0.4574 33.1289 RCIM_Linear>,<at 1 27.1289 RCIM_Linear>]}}
   }
   Trail controller - Bone: tongue_00(bAlphaBoolEnabled=True, RelaxationSpeedScale=1, FakeVelocity="1,1,1") {
    alphaInputType=Bool
    bActorSpaceFakeVel=true
    bLimitRotation=true
    bShowTrailLocation=true
    bUsePlanarLimit=true
    chainLength=5
    debugLifeTime=0.1
    lODThreshold=1
    maxDeltaTime=0.0333
    stretchLimit=2
    alphaScaleBiasClamp {bInterpResult=true, interpSpeedDecreasing=25, interpSpeedIncreasing=25}
    baseJoint {boneName=root}
    relaxationSpeedScaleInputProcessor {bias=0.3, scale=0.4}
    rotationLimits [
     {limitMax=<0 180 0>, limitMin=<0 -180 0>}
     {limitMax=<0 180 0>, limitMin=<0 -180 0>}
     {limitMax=<0 176.9904 0>, limitMin=<0 -180 0>}
     {limitMax=<0 180 0>, limitMin=<0 -180 0>}
     {limitMax=<180 180 180>}
    ]
    rotationOffsets [<0 0 0>, <0 0 0>, <0 0 0>, <0 0 0>, <0 0 0>]
    trailBone {boneName=tongue_00}
    trailRelaxationSpeed {editorCurveData {keys [<at 0 37.4961 RCIM_Linear>,<at 0.49 33.4961 RCIM_Linear>,<at 1 27.4961 RCIM_Linear>]}}
   }
   Trail controller - Bone: tongue_00!(ComponentPose, Alpha:float=1) -> (Pose) {
    bActorSpaceFakeVel=true
    bUsePlanarLimit=true
    chainLength=5
    lastBoneRotationAnimAlphaBlend=0.5
    maxDeltaTime=0.0333
    baseJoint {boneName=pack}
    planarLimits [{drivingBone {boneName=root}, planeTransform {location=<0 0 -40>, rotation=<rot -90 90 0>, scale=<0.2 1 1>}}]
    rotationLimits [
     {limitMax=<90 90 90>, limitMin=<-90 -90 -90>}
     {limitMax=<90 90 90>, limitMin=<-90 -90 -90>}
     {limitMax=<90 90 90>, limitMin=<-90 -90 -90>}
     {limitMax=<180 180 180>}
     {limitMax=<180 180 180>}
    ]
    rotationOffsets [<0 0 0>, <0 0 0>, <0 0 0>, <0 0 0>, <0 0 0>]
    trailBone {boneName=tongue_00}
    trailRelaxationSpeed {editorCurveData {keys [<at -0.01 20 RCIM_Linear>,<at 1 10 RCIM_Linear>]}}
   }
   Get Rigid Body Enabled() -> (Rigid Body Enabled:bool)
  ]}
 EventGraph {nodes [
   n1 = Cast To PlayerPawn_Athena(Object=TryGetPawnOwner())
   > Set Body Mesh(Body Mesh:Skeletal Mesh Component Object Reference=GetSkeletalMeshForPartType(self=@n1)) -> (Output_Get:Skeletal Mesh Component Object Reference)
   Event BlueprintUpdateAnimation()
   > Sequence!(execute) -> (then_0, then_1, then_2:Exec)
   n2 = GetOwnerSkeletalMeshComponent()
   n3 = BreakRotator(InRot=BreakTransform(InTransform=GetDeltaTransformFromRefPose(self=@n2, BoneName=spine_05, BaseName=spine_01)).Rotation)
   > Set Backpack Rotate(Backpack Rotate:Rotator=MakeRotator(Pitch=MapRangeClamped(Value=@n3.Yaw, InRangeA=0, InRangeB=-30, OutRangeB=125))) -> (Output_Get:Rotator)
   > Set Backpack Translation(Backpack Translation:Vector=MakeVector(Y=MapRangeClamped(Value=@n3.Yaw, InRangeA=0, InRangeB=-30, OutRangeB=-10))) -> (Output_Get:Vector)
   TryGetPawnOwner()
  ]}
]
variables [
 Backpack Rotate {type=Rotator}
 Backpack Translation {type=Vector}
 Body Mesh {type=SkeletalMeshComponent}
 Is Low LOD {type=boolean}
 Rigid Body Enabled {type=boolean}
 SkyDiveOff {type=boolean}
]
